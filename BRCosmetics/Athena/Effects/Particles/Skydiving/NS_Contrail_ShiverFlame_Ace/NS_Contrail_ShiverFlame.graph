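# Skydiving contrail cosmetic that draws a miniature road, anchored to a skeletal mesh bone, with barrels, barriers, dust and leg sand.
# The Road emitter computes a bend value that barrels, barriers and dust read via particle reads to follow curved offsets.
# ParaGlideLeanAlpha blends spawn rates, lifetimes and path shapes.
# Barriers die early, sending bend orientation packed as color in death events, alongside breakup parts, sand and debris emitters.
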NiagaraSystem /BRCosmetics/Athena/Effects/Particles/Skydiving/NS_Contrail_ShiverFlame_Ace/NS_Contrail_ShiverFlame
dependencies {
 usedDataInterfaces [
  &NiagaraDataInterfaceCurve
  &NiagaraDataInterfaceArrayFloat3
  &NiagaraDataInterfaceMeshRendererInfo
  &NiagaraDataInterfaceParticleRead
  &NiagaraDataInterfaceSpriteRendererInfo
  &NiagaraDataInterfaceVectorCurve
  &NiagaraDataInterfaceSkeletalMesh
  &NiagaraDataInterfaceArrayFloat
 ]
 usedDynamicInputs [
  &FN_FrontEndVector3Modifier
  &RotateVectorByQuaternion
  /Niagara/DynamicInputs/Rotation/InvertQuaternion
  &TransformVector
  &self:Abs_Z_Axis
  &NormalizeVector
 ]
 usedModules [
  &self:SystemSpawnScript.SystemScriptSource.SystemScriptGraph.NiagaraNodeAssignment_4.SetVariables_#1
  &self:SystemSpawnScript.SystemScriptSource.SystemScriptGraph.NiagaraNodeAssignment_5.SetVariables_#2
  /Game/Effects/Niagara/Modules/Skydiving/FN_DivingAttributes/v5/FN_DivingAttributes
  &self:SystemSpawnScript.SystemScriptSource.SystemScriptGraph.NiagaraNodeAssignment_1.SetVariables_#3
  &self:SystemSpawnScript.SystemScriptSource.SystemScriptGraph.NiagaraNodeAssignment_0.SetVariables_#4
  &SystemState
  &FN_Integrator
  &self:SystemSpawnScript.SystemScriptSource.SystemScriptGraph.NiagaraNodeAssignment_2.SetVariables_#5
  &self:SystemSpawnScript.SystemScriptSource.SystemScriptGraph.NiagaraNodeAssignment_3.SetVariables_#6
  &Timeline
 ]
}
emitters {
 Barrel002 {
  emitterUpdateScript [
   &EmitterState()
   &SpawnRate(
    SpawnRate=&FN_FrontEndFloatModifier(
     Gameplay=&Lerp_Float(A=1.25, B=1.8, Alpha=$User.ParaGlideLeanAlpha)
     Front End=&Lerp_Float(A=0.75, B=1.25, Alpha=$User.ParaGlideLeanAlpha)
    )
    Use Spawn Probability=false
   )
  ]
  particleSpawnScript [
   &InitializeParticle(
    Color Mode=::NewEnumerator0
    Lifetime=&Lerp_Float(A=3, B=2, Alpha=$User.ParaGlideLeanAlpha)
    Mesh Renderer Array Visibility Mode=::NewEnumerator3
    Mesh Renderer Info=DataInterface() {propertyValues {MeshRenderer=&self:Barrel002_1.NiagaraMeshRendererProperties_0}}
    Mesh Scale Mode=::NewEnumerator1
    Mesh Uniform Scale Max=1.1
    Mesh Uniform Scale Min=0.9
    Sprite Size Mode=::NewEnumerator3
    Uniform Sprite Size=72.8114
    UsePositionOffset=false
    Write Material Random=true
   )
   &InitialMeshOrientation(Mesh Orientation Mode=::NewEnumerator7, Rotation=<0.25 0 0>, Use Rotation Vector=true) {module=InitialMeshOrientation001}
   &InitialMeshOrientation(
    Mesh Orientation Mode=::NewEnumerator7
    Rotation=&RandomRangeVector(Minimum=<0 0 0>, Maximum=<0 1 0>)
    Use Rotation Vector=true
   )
   /SetVariables() {isSetParameters=true, module=SetVariables_#7}
  ]
  particleUpdateScript [
   &ParticleState()
   &ScaleMeshSize(Scale Factor=&VectorFromFloat(Value=&FloatFromCurve(FloatCurve=DataInterface() {propertyValues {Curve {keys [
          {interpMode=RCIM_Cubic, leaveTangentWeight=0.0333, tangentMode=RCTM_User, time=0, value=0}
          {arriveTangentWeight=0.0498, interpMode=RCIM_Cubic, leaveTangentWeight=0.3986, tangentMode=RCTM_User, time=0.3, value=1}
          {arriveTangentWeight=0.2667, interpMode=RCIM_Cubic, leaveTangentWeight=0.0333, tangentMode=RCTM_User, time=0.7, value=1}
          {tangentMode=RCTM_User, time=1, value=0}
         ]}}})))
   &self:road_bend_2(Particle Attribute Reader=DataInterface() {propertyValues {EmitterBinding {bindingMode=Other, emitterName=Road}}}) {module=road_bend_2}
   &SkeletalMeshLocation(
    Bone / Socket Index=5
    Bone Sampling Mode=::NewEnumerator4
    MaskSpawnGroup=false
    Sampled Position Offset=&Add_Vector(
     A=&Multiply_Vector(
      A=&Add_Vector(
       A=&Multiply_Vector(
        A=&Lerp_Vector(
         A=&VectorFromCurve(VectorCurve=DataInterface() {propertyValues {
            XCurve {keys [<at 0 250>,<at 1 -250>]}
            YCurve {keys [<at 0 -50 RCIM_Cubic>,<at 0.5 30 RCIM_Cubic>,<at 1 -50 RCIM_Cubic>]}
            ZCurve {keys [<at 0 100>,<at 1 100>]}
           }})
         B=&VectorFromCurve(VectorCurve=DataInterface() {propertyValues {
            XCurve {keys [<at 0 250>,<at 1 -250>]}
            YCurve {keys [<at 0 75 RCIM_Cubic>,<at 0.5 30 RCIM_Cubic>,<at 1 75 RCIM_Cubic>]}
            ZCurve {keys [<at 0 100>,<at 1 100>]}
           }})
         Alpha=$User.ParaGlideLeanAlpha
        )
        B=&MakeCustomVectorFromBool(
         Bool=&SetBoolByIntComparison(
          A=&ModuloInt(A=&ReturnParticleIDAttributes(), B=2)
          B=0
          Comparison Type=::NewEnumerator0
         )
         True Vector=<1 1 -1>
         False Vector=<1 1 1>
        )
       )
       B=&Lerp_Vector(
        A=&VectorFromCurve(VectorCurve=DataInterface() {propertyValues {
           XCurve {keys [<at 0 0>,<at 1 0>]}
           YCurve {keys [<at 0 0>,<at 1 0>]}
           ZCurve {keys [
             {
              arriveTangent=-2510.6699
              arriveTangentWeight=363.8912
              interpMode=RCIM_Cubic
              leaveTangent=-2510.6699
              leaveTangentWeight=363.8912
              tangentMode=RCTM_User
              tangentWeightMode=RCTWM_WeightedBoth
              time=0
              value=400
             }
             {
              arriveTangent=0.0001
              arriveTangentWeight=0.7224
              interpMode=RCIM_Cubic
              leaveTangent=0.0001
              leaveTangentWeight=0.7224
              tangentMode=RCTM_User
              tangentWeightMode=RCTWM_WeightedBoth
              time=1
              value=0
             }
            ]}
          }})
        B=&VectorFromCurve(
         VectorCurve=DataInterface() {propertyValues {
           XCurve {keys [<at 0 0>,<at 1 0>]}
           YCurve {keys [<at 0 0>,<at 1 0>]}
           ZCurve {keys [
             {
              arriveTangent=-2510.6699
              arriveTangentWeight=363.8912
              interpMode=RCIM_Cubic
              leaveTangent=-2510.6699
              leaveTangentWeight=363.8912
              tangentMode=RCTM_User
              tangentWeightMode=RCTWM_WeightedBoth
              time=0
              value=400
             }
             {
              arriveTangent=0.0001
              arriveTangentWeight=0.7224
              interpMode=RCIM_Cubic
              leaveTangent=0.0001
              leaveTangentWeight=0.7224
              tangentMode=RCTM_User
              tangentWeightMode=RCTWM_WeightedBoth
              time=1
              value=0
             }
            ]}
          }}
         Scale Curve=<1 1 -1>
        )
        Alpha=&Lerp_Float(
         A=0.5
         B=&RemapRange(Input Value To Remap=$Particles.Roadbending, Input Min=-250, Input Max=250, Output Min=0, Output Max=1)
         Alpha=1
        )
       )
      )
      B=$System.MasterRoadScale
     )
     B=$System.MasterRoadOffset
    )
    Skeletal Mesh=$System.Skeletal Mesh
   ) {module=SkeletalMeshLocation001}
   &DynamicMaterialParameters(
    Index 0 Param 1=$Particles.Roadbending
    Module.Index0Param0Write=true
    Module.Index0Param1Write=true
    Module.Index0Param2Write=true
    Module.Index0Param3Write=true
   )
  ]
  rendererClasses [&NiagaraMeshRendererProperties]
  renderers [{rendererClass=&NiagaraMeshRendererProperties, rendererIndex=0}]
 }
 Barrior {
  emitterUpdateScript [
   &EmitterState()
   &SpawnRate(SpawnRate=3, Use Spawn Probability=false) {enabled=false, module=SpawnRate001}
   &SpawnRate(SpawnRate=&Lerp_Float(A=0.3333, B=0.6667, Alpha=$User.ParaGlideLeanAlpha), Use Spawn Probability=false)
  ]
  particleSpawnScript [
   &InitializeParticle(
    Color Mode=::NewEnumerator0
    Lifetime=&Lerp_Float(A=3, B=2, Alpha=$User.ParaGlideLeanAlpha)
    Mesh Scale Mode=::NewEnumerator1
    Mesh Uniform Scale Max=1.1
    Mesh Uniform Scale Min=0.9
    Sprite Size Mode=::NewEnumerator3
    Uniform Sprite Size=72.8114
    UsePositionOffset=false
    Write Material Random=true
   )
   &InitialMeshOrientation(Mesh Orientation Mode=::NewEnumerator7, Rotation=<0.25 0 0>, Use Rotation Vector=true) {module=InitialMeshOrientation001}
   &InitialMeshOrientation(
    Mesh Orientation Mode=::NewEnumerator7
    Rotation=&RandomRangeVector(Minimum=<0 -0.25 0>, Maximum=<0 -0.25 0>)
    Use Rotation Vector=true
   )
   /SetVariables() {isSetParameters=true, module=SetVariables_#7}
  ]
  particleUpdateScript [
   &ParticleState()
   &SolveForcesAndVelocity(Clamp Velocity=false, Limit Acceleration=false)
   &ScaleMeshSize(Scale Factor=&VectorFromFloat(Value=&FloatFromCurve(FloatCurve=DataInterface() {propertyValues {Curve {keys [
          {interpMode=RCIM_Cubic, leaveTangentWeight=0.0333, tangentMode=RCTM_User, time=0, value=0}
          {arriveTangentWeight=0.0498, interpMode=RCIM_Cubic, leaveTangentWeight=0.3986, tangentMode=RCTM_User, time=0.1, value=1}
          {arriveTangentWeight=0.2667, interpMode=RCIM_Cubic, leaveTangentWeight=0.0333, tangentMode=RCTM_User, time=0.7, value=1}
          {tangentMode=RCTM_User, time=1, value=0}
         ]}}})))
   &self:road_bend_2(Particle Attribute Reader=DataInterface() {propertyValues {EmitterBinding {bindingMode=Other, emitterName=Road}}}) {module=road_bend_2}
   &SkeletalMeshLocation(
    Bone / Socket Index=5
    Bone Sampling Mode=::NewEnumerator4
    MaskSpawnGroup=false
    Sampled Position Offset=&Add_Vector(
     A=&Multiply_Vector(
      A=&Add_Vector(
       A=&Multiply_Vector(
        A=&Lerp_Vector(
         A=&VectorFromCurve(
          VectorCurve=DataInterface() {propertyValues {
            XCurve {keys [<at 0 250>,<at 1 -250>]}
            YCurve {keys [<at 0 -50 RCIM_Cubic>,<at 0.5 30 RCIM_Cubic>,<at 1 -50 RCIM_Cubic>]}
            ZCurve {keys [<at 0 0>,<at 1 0>]}
           }}
          Scale Curve=<1 1 100>
         )
         B=&VectorFromCurve(VectorCurve=DataInterface() {propertyValues {
            XCurve {keys [<at 0 250>,<at 1 -250>]}
            YCurve {keys [<at 0 75 RCIM_Cubic>,<at 0.5 30 RCIM_Cubic>,<at 1 75 RCIM_Cubic>]}
            ZCurve {keys [<at 0 0>,<at 1 0>]}
           }})
         Alpha=$User.ParaGlideLeanAlpha
        )
        B=<1 1 1>
       )
       B=&Lerp_Vector(
        A=&VectorFromCurve(VectorCurve=DataInterface() {propertyValues {
           XCurve {keys [<at 0 0>,<at 1 0>]}
           YCurve {keys [<at 0 0>,<at 1 0>]}
           ZCurve {keys [
             {
              arriveTangent=-2510.6699
              arriveTangentWeight=363.8912
              interpMode=RCIM_Cubic
              leaveTangent=-2510.6699
              leaveTangentWeight=363.8912
              tangentMode=RCTM_User
              tangentWeightMode=RCTWM_WeightedBoth
              time=0
              value=400
             }
             {
              arriveTangent=0.0001
              arriveTangentWeight=0.7224
              interpMode=RCIM_Cubic
              leaveTangent=0.0001
              leaveTangentWeight=0.7224
              tangentMode=RCTM_User
              tangentWeightMode=RCTWM_WeightedBoth
              time=1
              value=0
             }
            ]}
          }})
        B=&VectorFromCurve(
         VectorCurve=DataInterface() {propertyValues {
           XCurve {keys [<at 0 0>,<at 1 0>]}
           YCurve {keys [<at 0 0>,<at 1 0>]}
           ZCurve {keys [
             {
              arriveTangent=-2510.6699
              arriveTangentWeight=363.8912
              interpMode=RCIM_Cubic
              leaveTangent=-2510.6699
              leaveTangentWeight=363.8912
              tangentMode=RCTM_User
              tangentWeightMode=RCTWM_WeightedBoth
              time=0
              value=400
             }
             {
              arriveTangent=0.0001
              arriveTangentWeight=0.7224
              interpMode=RCIM_Cubic
              leaveTangent=0.0001
              leaveTangentWeight=0.7224
              tangentMode=RCTM_User
              tangentWeightMode=RCTWM_WeightedBoth
              time=1
              value=0
             }
            ]}
          }}
         Scale Curve=<1 1 -1>
        )
        Alpha=&Lerp_Float(
         A=0.5
         B=&RemapRange(Input Value To Remap=$Particles.Roadbending, Input Min=-250, Input Max=250, Output Min=0, Output Max=1)
         Alpha=1
        )
       )
      )
      B=$System.MasterRoadScale
     )
     B=$System.MasterRoadOffset
    )
    Skeletal Mesh=$System.Skeletal Mesh
   ) {module=SkeletalMeshLocation001}
   &DynamicMaterialParameters(
    Index 0 Param 1=$Particles.Roadbending
    Module.Index0Param0Write=true
    Module.Index0Param1Write=true
    Module.Index0Param2Write=true
    Module.Index0Param3Write=true
   )
   &KillParticles(Kill Particles=&SetBoolByFloatComparison(A=$Particles.NormalizedAge, B=&FN_FrontEndFloatModifier(Gameplay=0.35, Front End=0.25)))
   /SetVariables(Particles.Temp Mesh Orient:Quat4f=$Particles.MeshOrientation) {isSetParameters=true}
   /SetVariables(Particles.RoadBendOrientation:Quat4f=&MultiplyQuaternion(
     Quaternion A=&MakeQuatFromAxisAngle(
      Axis=&ReturnMeshOrientationAxis(Mesh Orientation Axis=::ZAxis, Mesh Orientation Quat=$Particles.MeshOrientation)
      AngleInDegrees=&Multiply_Float(
       A=&Lerp_Float(
        A=90
        B=-90
        Alpha=&RemapRange(Input Value To Remap=$Particles.Roadbending, Input Min=-250, Input Max=250, Output Min=0, Output Max=1)
       )
       B=&FloatFromCurve(
        FloatCurve=DataInterface() {propertyValues {Curve {keys [
            {
             arriveTangent=-2510.6699
             arriveTangentWeight=363.8912
             interpMode=RCIM_Cubic
             leaveTangent=-2510.6699
             leaveTangentWeight=363.8912
             tangentMode=RCTM_User
             tangentWeightMode=RCTWM_WeightedBoth
             time=0
             value=400
            }
            {
             arriveTangent=0.0001
             arriveTangentWeight=0.7224
             interpMode=RCIM_Cubic
             leaveTangent=0.0001
             leaveTangentWeight=0.7224
             tangentMode=RCTM_User
             tangentWeightMode=RCTWM_WeightedBoth
             time=1
             value=0
            }
           ]}}}
        Scale Curve=0.0025
       )
      )
     )
     Quaternion B=$Particles.MeshOrientation
    )) {isSetParameters=true}
   &GenerateDeathEvent(
    "Color to Send as Linear Color (ParticleColor":LinearColor=&self:ColorToQuat(Quat:Quat4f=$Particles.RoadBendOrientation)
    Use Event Probability=false
   )
  ]
  rendererClasses [&NiagaraMeshRendererProperties, &NiagaraSpriteRendererProperties]
  renderers [
   {rendererClass=&NiagaraMeshRendererProperties, rendererIndex=0}
   {rendererClass=&NiagaraSpriteRendererProperties, rendererIndex=1}
  ]
 }
 Barrior_Parts_Burst {
  emitterUpdateScript [
   /SetVariables(Emitter.Part Position offsets=DataInterface() {propertyValues {
      GpuSyncMode=SyncCpuToGpu
      MaxElements=0
      FloatData [<-48.303 -0.129 39.13>, <47.038 -0.105 40.829>, <-27.998 3.22 59.775>, <23.343 3.22 66.66>, <-20.932 3.245 32.978>, <25.298 3.275 32.978>]
     }}) {isSetParameters=true}
   &EmitterState()
   &SpawnBurst_Instantaneous(Spawn Count=7, Use Spawn Probability=false, UseLoopCountLimit=false) {enabled=false}
  ]
  particleSpawnScript [
   &InitializeParticle(
    Color Mode=::NewEnumerator0
    Lifetime Max=0.35
    Lifetime Min=0.3
    Lifetime Mode=::NewEnumerator1
    Mesh Renderer Array Visibility Mode=::NewEnumerator1
    Mesh Renderer Info=DataInterface() {propertyValues {MeshRenderer=&self:Barrior_Parts_Burst_2.NiagaraMeshRendererProperties_0}}
    Mesh Renderer Mesh Index:NiagaraInt32=&ReturnParticleIDAttributes()
    Mesh Scale Mode=::NewEnumerator3
    Position Offset=&SelectVectorFromArray(
     Direct Array Index=&ReturnParticleIDAttributes()
     Direct Array Mode=::NewEnumerator1
     Vector Selection Array=$Emitter.Part Position offsets
    )
    UsePositionOffset=true
    Write Material Random=true
   )
   &AddVelocity(
    Module.Use Falloff Distance=false
    Origin Offset=<0 0 50>
    Override Seed=false
    Velocity Mode=::NewEnumerator1
    Velocity Speed=&RandomRangeFloat(Minimum=300, Maximum=500)
   ) {enabled=false}
   &AddVelocity(Velocity=&RandomVector(Vector Scale=1), Velocity Speed Scale=350) {enabled=false, module=AddVelocity001}
   &InitialMeshOrientation(Mesh Orientation Mode=::NewEnumerator7, Rotation=<0.25 0 0>, Use Rotation Vector=true) {enabled=false}
  ]
  particleUpdateScript [
   &ParticleState()
   &ScaleMeshSize(Scale Factor=&VectorFromFloat(Value=&FloatFromCurve(FloatCurve=DataInterface() {propertyValues {Curve {keys [<at 0.8 1>,<at 1 0>]}}})))
   &UpdateMeshOrientation(
    Rotation Rate=&RandomRangeFloat(Minimum=0.5, Maximum=1)
    Rotation Vector=&RandomVector(Vector Scale=1)
   )
   &InheritSourceMovement(
    Applied Position Scale=&VectorFromFloat(Value=&FloatFromCurve(FloatCurve=DataInterface() {propertyValues {Curve {keys [
          {
           arriveTangentWeight=1.1661
           interpMode=RCIM_Cubic
           leaveTangentWeight=1.1661
           tangentMode=RCTM_User
           tangentWeightMode=RCTWM_WeightedBoth
           time=0
           value=1
          }
          {
           arriveTangent=-5.7781
           arriveTangentWeight=0.7354
           interpMode=RCIM_Cubic
           leaveTangent=-5.7781
           leaveTangentWeight=0.7354
           tangentMode=RCTM_User
           tangentWeightMode=RCTWM_WeightedBoth
           time=1
           value=0
          }
         ]}}}))
    ApplyForce=false
    ApplyPosition=true
    ApplyPositionFalloff=false
   )
   &AccelerationForce(
    Acceleration=&Multiply_VectorByFloat(
     Vector=&Lerp_Vector(A=<0 0 -980>, B=<1 -980 1>, Alpha=$User.ParaGlideLeanAlpha)
     Float=&Lerp_Float(
      A=&FloatFromCurve(
       FloatCurve=DataInterface() {propertyValues {Curve {keys [
           {
            arriveTangentWeight=0.6755
            interpMode=RCIM_Cubic
            leaveTangentWeight=0.6755
            tangentMode=RCTM_User
            tangentWeightMode=RCTWM_WeightedBoth
            time=0
            value=0
           }
           {arriveTangent=1.9684, interpMode=RCIM_Cubic, leaveTangent=1.9684, tangentMode=RCTM_User, time=1, value=1}
          ]}}}
       Scale Curve=15
      )
      B=&FloatFromCurve(
       FloatCurve=DataInterface() {propertyValues {Curve {keys [
           {
            arriveTangentWeight=0.6755
            interpMode=RCIM_Cubic
            leaveTangentWeight=0.6755
            tangentMode=RCTM_User
            tangentWeightMode=RCTWM_WeightedBoth
            time=0
            value=0
           }
           {arriveTangent=1.9684, interpMode=RCIM_Cubic, leaveTangent=1.9684, tangentMode=RCTM_User, time=1, value=1}
          ]}}}
       Scale Curve=10
      )
      Alpha=$User.ParaGlideLeanAlpha
     )
    )
    Coordinate Space=::Local
   )
   &Drag(Drag=3, Use Linear Drag=true, Use Rotational Drag=true)
   &GravityForce() {enabled=false}
   &SolveForcesAndVelocity(Clamp Velocity=false, Limit Acceleration=false)
  ]
  rendererClasses [&NiagaraSpriteRendererProperties, &NiagaraMeshRendererProperties]
  renderers [
   {rendererClass=&NiagaraSpriteRendererProperties}
   {rendererClass=&NiagaraMeshRendererProperties, rendererIndex=1}
  ]
 }
 Barrior_death_sand {
  emitterUpdateScript [
   &EmitterState(Life Cycle Mode=::NewEnumerator1, Loop Duration=1, UseLoopDelay=false)
   &SpawnBurst_Instantaneous(Spawn Count=7, Use Spawn Probability=false, UseLoopCountLimit=false) {enabled=false}
  ]
  particleSpawnScript [
   &InitializeParticle(
    Color Mode=::NewEnumerator0
    Lifetime=0.5
    Mesh Scale Mode=::NewEnumerator3
    Sprite Rotation Mode=::NewEnumerator2
    Sprite Size Mode=::NewEnumerator3
    Uniform Sprite Size=0
    UsePositionOffset=true
    Write Material Random=true
   )
   &AddVelocity(
    Module.Use Falloff Distance=false
    Origin Offset=<0 0 50>
    Override Seed=false
    Velocity Mode=::NewEnumerator1
    Velocity Speed=&Multiply_Float(A=&RandomRangeFloat(Minimum=-300, Maximum=-500), B=0.2)
   )
   &AddVelocity(Velocity=&RandomVector(Vector Scale=1), Velocity Speed Scale=150) {module=AddVelocity001}
   &SubUVAnimation(
    "Sprite Renderer "=DataInterface() {propertyValues {SpriteRenderer=&self:Barrior_death_sand_20.NiagaraSpriteRendererProperties_1}}
    SubUV Animation Mode:ENiagara_SubUVLookupModeV2=::NewEnumerator2
    UseEndFrame=true
    UseStartFrame=true
   ) {enabled=false}
   &CameraOffset(Camera Offset Amount=15)
  ]
  particleUpdateScript [
   &ParticleState()
   &SpriteRotationRate(Rotation Rate=&Multiply_Float(A=90, B=&RandomRangeFloat(Minimum=-1, Maximum=1)))
   &ScaleSpriteSize(
    Scale Sprite Size Mode=::NewEnumerator0
    Uniform Scale Factor=&FloatFromCurve(FloatCurve=DataInterface() {propertyValues {Curve {keys [{tangentMode=RCTM_User, time=0.5, value=0}, {tangentMode=RCTM_User, time=1, value=1}]}}})
   )
   &UpdateMeshOrientation(
    Rotation Rate=&RandomRangeFloat(Minimum=0.5, Maximum=1)
    Rotation Vector=&RandomVector(Vector Scale=1)
   )
   &InheritSourceMovement(
    Applied Position Scale=&VectorFromFloat(Value=&FloatFromCurve(FloatCurve=DataInterface() {propertyValues {Curve {keys [<at 0 1 RCIM_Cubic>, {arriveTangent=-0.0499, interpMode=RCIM_Cubic, leaveTangent=-0.0499, tangentMode=RCTM_User, time=1, value=0.98}]}}}))
    ApplyForce=false
    ApplyPosition=true
    ApplyPositionFalloff=false
   )
   &SolveForcesAndVelocity(Clamp Velocity=false, Limit Acceleration=false)
   /SetVariables(Particles.SpriteSize=&Vector2DFromFloat(Value=&FloatFromCurve(
      FloatCurve=DataInterface() {propertyValues {Curve {keys [
          {arriveTangent=1.9242, interpMode=RCIM_Cubic, leaveTangent=1.9242, tangentMode=RCTM_User, time=0, value=0}
          {arriveTangent=0.5536, interpMode=RCIM_Cubic, leaveTangent=0.5536, tangentMode=RCTM_User, time=1, value=1}
         ]}}}
      Scale Curve=250
     ))) {isSetParameters=true}
   &Color(
    Color=&RandomRangeLinearColor(Minimum=<rgba 0.2 0.1242 0.07 0.4>, Maximum=<rgba 0.25 0.1713 0.0865 0.25>)
    Scale Alpha=&FloatFromCurve(FloatCurve=DataInterface() {propertyValues {Curve {keys [
         {arriveTangent=9.0725, interpMode=RCIM_Cubic, leaveTangent=9.0725, tangentMode=RCTM_User, time=0, value=0}
         {arriveTangent=-2.5249, interpMode=RCIM_Cubic, leaveTangent=-2.5249, tangentMode=RCTM_User, time=0.1408, value=0.9847}
         {arriveTangent=-0.0405, interpMode=RCIM_Cubic, leaveTangent=-0.0405, tangentMode=RCTM_User, time=1, value=0}
        ]}}})
    Scale Color=&VectorFromFloat(Value=1)
   )
   &SubUVAnimation(
    Random Start Frame=true
    "Sprite Renderer "=DataInterface() {propertyValues {SpriteRenderer=&self:Barrior_death_sand_20.NiagaraSpriteRendererProperties_0}}
    SubUV Lookup Index:NiagaraFloat=$Particles.Age
    SubUV Lookup Index Scale=0.35
    UseEndFrame=false
    UseStartFrame=false
   ) {module=SubUVAnimation002}
   /SetVariables() {isSetParameters=true}
   /SetVariables(
    Particles.Sand:LinearColor=&MultiplyLinearColor(
     Linear Color=$Particles.Color
     Scale Factor=&MakeFloat4FromVectorAndFloat("Vector (RGB)"=&VectorFromFloat(Value=0.5), "Float (Alpha)"=2)
    )
    Particles.sand Size:Vector2f=&Multiply_Vector2D(A=$Particles.SpriteSize, B=&Vector2DFromFloat(Value=1.5))
   ) {isSetParameters=true}
  ]
  rendererClasses [&NiagaraSpriteRendererProperties]
  renderers [
   {rendererClass=&NiagaraSpriteRendererProperties}
   {rendererClass=&NiagaraSpriteRendererProperties, rendererIndex=1}
   {rendererClass=&NiagaraSpriteRendererProperties, rendererIndex=2}
   {rendererClass=&NiagaraSpriteRendererProperties, rendererIndex=3}
  ]
 }
 Empty {
  bEnabled=false
  emitterUpdateScript [
   &EmitterState()
   &SpawnBurst_Instantaneous(Use Spawn Probability=false, UseLoopCountLimit=false)
  ]
  particleSpawnScript [&InitializeParticle(Color Mode=::NewEnumerator0, Lifetime=5, UsePositionOffset=false, Write Material Random=true)]
  particleUpdateScript [&ParticleState()]
  rendererClasses [&NiagaraSpriteRendererProperties, &NiagaraMeshRendererProperties]
  renderers [
   {rendererClass=&NiagaraSpriteRendererProperties}
   {rendererClass=&NiagaraMeshRendererProperties, rendererIndex=1}
  ]
 }
 Leg_Sand {
  emitterUpdateScript [
   &EmitterState()
   &SpawnRate(SpawnRate=&Multiply_Float(A=30, B=2), Use Spawn Probability=false)
  ]
  particleSpawnScript [
   &InitializeParticle(
    Color Maximum=<rgba 0.2 0.1242 0.07 0.35>
    Color Minimum=<rgba 0.4 0.2741 0.1384 0.5>
    Color Mode=::NewEnumerator2
    Lifetime Max=0.45
    Lifetime Min=0.35
    Lifetime Mode=::NewEnumerator1
    Sprite Rotation Mode=::NewEnumerator2
    Sprite Size Mode=::NewEnumerator3
    Uniform Sprite Size=80
    UsePositionOffset=false
    Write Material Random=true
   )
   &PartitionParticles()
   &SkeletalMeshLocation(
    Bone / Socket Index=&Add_Integer(A=$Particles.PartitionParticles.Partition, B=2)
    Bone Sampling Mode=::NewEnumerator4
    MaskSpawnGroup=false
    Orientation Sampling=::NewEnumerator1
    Skeletal Mesh=$System.Skeletal Mesh
   )
   &SphereLocation(
    MaskSpawnGroup=false
    Override Seed=false
    Sphere Radius=3
    Surface Only=false
    UseNonUniformScale=false
    UseOffset=false
    UseSphereOrientation=false
   )
   &AddVelocity(
    Coordinate Space=::World
    Velocity=&Multiply_Vector(
     A=&Lerp_Vector(
      A=&Multiply_VectorByFloat(Vector=$Engine.Owner.SystemYAxis, Float=-1)
      B=&Multiply_VectorByFloat(Vector=$Engine.Owner.SystemZAxis, Float=1)
      Alpha=$User.ParaGlideLeanAlpha
     )
     B=&VectorFromFloat(Value=300)
    )
   )
   &SubUVAnimation(
    "Sprite Renderer "=DataInterface() {propertyValues {SpriteRenderer=&self:Leg_Sand_22.NiagaraSpriteRendererProperties_0}}
    UseEndFrame=false
    UseStartFrame=false
   ) {enabled=false}
   &CameraOffset(Camera Offset Amount=10)
   &AddVelocity(Velocity=&Lerp_Vector(A=<0 0 0>, B=<0 -150 0>, Alpha=$User.ParaGlideLeanAlpha)) {module=AddVelocity001}
  ]
  particleUpdateScript [
   &ParticleState()
   &SpriteRotationRate(Rotation Rate=&MakeCustomFloatFromBool(
     Bool=&SetBoolByIntComparison(A=$Particles.PartitionParticles.Partition, B=0, Comparison Type=::NewEnumerator4)
     True Float=&RandomRangeFloat(Minimum=-300, Maximum=-360)
     False Float=&RandomRangeFloat(Minimum=300, Maximum=360)
    )) {enabled=false}
   &SampleSkeletalMesh(
    AdditionalYawPitchRoll=false
    Bone / Socket Index:NiagaraInt32=&Add_Integer(A=$Particles.PartitionParticles.Partition, B=2)
    Bone Sampling Mode:ENiagaraBoneSamplingMode=::NewEnumerator4
    Skeletal Mesh:NiagaraDataInterfaceSkeletalMesh=$System.Skeletal Mesh
   )
   &InheritSourceMovement(ApplyForce=false, ApplyPosition=true, ApplyPositionFalloff=false, Source Velocity:Vector3f=$Particles.SampleSkeletalMesh.SampledVelocity)
   &InheritSourceMovement(
    Applied Position Scale=&Multiply_VectorByFloat(
     Vector=<1 1 0>
     Float=&FloatFromCurve(FloatCurve=DataInterface() {propertyValues {Curve {keys [<at 0 1>,<at 1 0>]}}})
    )
    ApplyForce=false
    ApplyPosition=true
    ApplyPositionFalloff=false
    Source Velocity:Vector3f=$Particles.SampleSkeletalMesh.SampledVelocity
   ) {enabled=false, module=InheritSourceMovement001}
   &SolveForcesAndVelocity(Clamp Velocity=false, Limit Acceleration=false)
   &ScaleColor(
    Scale Alpha=&Multiply_Float(
     A=&FloatFromCurve(
      FloatCurve=DataInterface() {propertyValues {Curve {keys [<at 0 0 RCIM_Cubic>,<at 0 0 RCIM_Cubic>,<at 0.25 1>,<at 1 0 RCIM_Cubic>]}}}
      Scale Curve=0.25
     )
     B=&FNTimeOfDayScalar(Night Scale=0.5, Morning Scale=1, Evening Scale=1)
    )
    ScaleA=true
    ScaleRGB=true
   )
   &ScaleSpriteSize(Uniform Curve Sprite Scale=DataInterface() {propertyValues {Curve {keys [{arriveTangent=1.2187, time=0, value=0.2}, {leaveTangent=3.3905, time=1, value=1.2}]}}})
   /SetVariables(
    Particles.Sand:LinearColor=&MultiplyLinearColor(
     Linear Color=$Particles.Color
     Scale Factor=&MakeFloat4FromVectorAndFloat("Vector (RGB)"=&VectorFromFloat(Value=1), "Float (Alpha)"=2)
    )
    Particles.sand Size:Vector2f=&Multiply_Vector2D(A=$Particles.SpriteSize, B=&Vector2DFromFloat(Value=1.25))
   ) {isSetParameters=true}
   &SubUVAnimation(
    Random Start Frame=true
    "Sprite Renderer "=DataInterface() {propertyValues {SpriteRenderer=&self:Leg_Sand_22.NiagaraSpriteRendererProperties_0}}
    SubUV Lookup Index:NiagaraFloat=$Particles.Age
    SubUV Lookup Index Scale=0.35
    UseEndFrame=false
    UseStartFrame=false
   ) {module=SubUVAnimation001}
   /SetVariables() {isSetParameters=true}
  ]
  rendererClasses [&NiagaraSpriteRendererProperties]
  renderers [
   {rendererClass=&NiagaraSpriteRendererProperties}
   {rendererClass=&NiagaraSpriteRendererProperties, rendererIndex=1}
   {rendererClass=&NiagaraSpriteRendererProperties, rendererIndex=2}
   {rendererClass=&NiagaraSpriteRendererProperties, rendererIndex=3}
  ]
 }
 NE_FN_PhysMatDebris {
  emitterUpdateScript [
   &EmitterState()
   &SpawnRate(SpawnRate=300, Use Spawn Probability=false) {enabled=false}
   &SpawnBurst_Instantaneous(
    Spawn Count=&RandomRangeInt(Minimum=5, Maximum=10)
    Use Spawn Probability=false
    UseLoopCountLimit=false
   ) {enabled=false}
  ]
  particleSpawnScript [
   /SetVariables(Particles.BiasedRandom=&FloatFromCurve(
     FloatCurve=DataInterface() {propertyValues {Curve {keys [
         <at 0.2 1 RCIM_Cubic>
         {arriveTangent=-3.3394, interpMode=RCIM_Cubic, leaveTangent=-3.3394, tangentMode=RCTM_User, time=0.5, value=0.5}
         <at 0.8 0 RCIM_Cubic>
        ]}}}
     CurveIndex=&RandomRangeFloat(Minimum=0, Maximum=1)
    )) {isSetParameters=true}
   &InitializeParticle(
    Color=<rgba 4 4 4 1>
    Lifetime=&RandomRangeFloat(Minimum=0.65, Maximum=0.35)
    Mass=&Lerp_Float(A=0.6, B=1.2, Alpha=$Particles.BiasedRandom)
    Mass Mode=::NewEnumerator1
    Mesh Scale=&Multiply_VectorByFloat(
     Vector=&RandomRangeVector(Minimum=<0.9 0.4 0.5>, Maximum=<1 1 1>)
     Float=&Lerp_Float(A=0.7, B=1, Alpha=$Particles.BiasedRandom)
    )
    Mesh Scale Mode=::NewEnumerator4
    UsePositionOffset=false
    Write Material Random=true
   )
   &FN_PhysicalSurface(
    Dirt Mesh Count=2
    Glass Mesh Count=3
    Ice Mesh Count=2
    Metal Gravity=&RandomRangeFloat(Minimum=-350, Maximum=-1800)
    Metal Mesh Count=2
    Metal Velocity=&RandomRangeFloat(Minimum=80, Maximum=375)
    Metal Velocity Cone Angle=130
    Physical Surface=1
    Stone Gravity=&RandomRangeFloat(Minimum=-700, Maximum=-1250)
    Stone Mesh Count=2
    Stone Velocity=&RandomRangeFloat(Minimum=80, Maximum=550)
    Stone Velocity Cone Angle=120
    Wood Color:LinearColor=<rgba 3 3 3 1>
    Wood Gravity=&RandomRangeFloat(Minimum=0, Maximum=0)
    Wood Mesh Count=2
    Wood Scale=1.5
    Wood Velocity=&RandomRangeFloat(Minimum=80, Maximum=600)
    Wood Velocity Cone Angle=130
   ) {enabled=false}
   &AddVelocityInCone(
    Cone Angle=$Output.FN_PhysicalSurface.SurfaceResponseConeAngle
    Module.Use Velocity Falloff On Cone Axis=false
    Velocity Strength=$Output.FN_PhysicalSurface.SurfaceResponseVelocity
   ) {enabled=false}
   &InitialMeshOrientation(Use Rotation Vector=false)
   &AddRotationalVelocity(Rotation Rate=&RandomVector(Vector Scale=&RandomRangeFloat(Minimum=1, Maximum=2)))
   &ApplyInitialForces()
   /SetVariables(Particles.GravityOverrideValue=$Output.FN_PhysicalSurface.GravityOverride) {enabled=false, isSetParameters=true}
  ]
  particleUpdateScript [
   &ParticleState()
   &GravityForce(Gravity=$Particles.GravityOverrideValue) {enabled=false}
   &Collision(
    Advanced Aging Rate=1
    Age Colliding Particles=false
    CPU Collision Trace Channel=ECC_WorldStatic
    Collision Enabled=&Boolean_AndOperation(
     Bool1=&SetBoolByFloatComparison(A=$Particles.Initial.Lifetime, B=1)
     Bool2=&SetBoolByFloatComparison(A=$Particles.Age, B=0.12)
    )
    Control Roll On Collision=true
    Friction=1
    Friction During a Bounce=0.25
    Max CPU Trace Length=100
    Module.Control Rest State Via Surface Normal=false
    Module.EnableMaxCollisionCount=false
    Module.Randomize Collision Normal Vector=true
    Randomize Collision Normal=0.25
    ScaleRateOfRotation=false
   ) {enabled=false}
   &InheritSourceMovement(
    Applied Position Scale=&VectorFromFloat(Value=1)
    ApplyForce=false
    ApplyPosition=true
    ApplyPositionFalloff=false
   )
   &SolveForcesAndVelocity(Clamp Velocity=false, Limit Acceleration=false)
   &ScaleMeshSize(Scale Factor=&VectorFromFloat(Value=&FloatFromCurve(FloatCurve=DataInterface() {propertyValues {Curve {keys [<at 0.6 1 RCIM_Cubic>, {arriveTangent=-6.308, interpMode=RCIM_Cubic, leaveTangent=-6.308, tangentMode=RCTM_User, time=1, value=0}]}}})))
   &DynamicMaterialParameters(Module.Param0WriteEnabled=true, Module.Param1WriteEnabled=true, Module.Param2WriteEnabled=false, Module.Param3WriteEnabled=false)
  ]
  rendererClasses [&NiagaraMeshRendererProperties]
  renderers [{rendererClass=&NiagaraMeshRendererProperties, rendererIndex=0}]
 }
 Road {
  emitterUpdateScript [
   &EmitterState(Life Cycle Mode=::NewEnumerator1, Loop Behavior=::NewEnumerator1, Loop Duration=1.6644, UseLoopDelay=false)
   &SpawnBurst_Instantaneous(Use Spawn Probability=false, UseLoopCountLimit=false)
   &SpawnBurst_Instantaneous(Spawn Time=1.6644, Use Spawn Probability=false, UseLoopCountLimit=false) {module=SpawnBurst_Instantaneous001}
   /SetVariables(Emitter.PanningFloat:NiagaraFloat=&Add_Float(
     A=$Emitter.PanningFloat
     B=&Multiply_Float(A=$Engine.DeltaTime, B=&Lerp_Float(A=-0.5, B=-1, Alpha=$User.ParaGlideLeanAlpha))
    )) {isSetParameters=true}
  ]
  particleSpawnScript [
   &InitializeParticle(
    Color Mode=::NewEnumerator0
    Lifetime=5
    Mesh Scale=&Multiply_Vector(
     A=<0.65 1 1>
     B=&self:SwizzleRoad(Default:Vector3f=$System.MasterRoadScale)
    )
    Mesh Scale Mode=::NewEnumerator4
    Sprite Size Mode=::NewEnumerator3
    Uniform Sprite Size=550
    UsePositionOffset=false
    Write Material Random=true
   )
   &SkeletalMeshLocation(
    Bone / Socket Index=5
    Bone Sampling Mode=::NewEnumerator4
    MaskSpawnGroup=false
    Orientation Sampling=::NewEnumerator1
    Sampled Position Offset=<10 35 0>
    Skeletal Mesh=$System.Skeletal Mesh
   ) {module=SkeletalMeshLocation001}
   /SetVariables(Particles.MeshOrientation=&MultiplyQuaternion(
     Quaternion A=&MultiplyQuaternion(
      Quaternion A=$Particles.SkeletalMeshLocation001.SampledRotation
      Quaternion B=&MakeQuatFromAxisAngle(Axis=<0 1 0>, AngleInDegrees=90)
     )
     Quaternion B=&MakeQuatFromAxisAngle(Axis=<1 0 0>, AngleInDegrees=90)
    )) {isSetParameters=true}
   /SetVariables() {isSetParameters=true}
  ]
  particleUpdateScript [
   &ParticleState(Kill Particles When Lifetime Has Elapsed=false)
   &ScaleColor(
    Scale Alpha=&FloatFromCurve(
     FloatCurve=DataInterface() {propertyValues {Curve {keys [{interpMode=RCIM_Cubic, tangentMode=RCTM_User, time=0, value=0}, {interpMode=RCIM_Cubic, tangentMode=RCTM_User, time=1, value=1}]}}}
     CurveIndex=$Emitter.Age
    )
    ScaleA=true
    ScaleRGB=true
   )
   /SetVariables(Particles.PreviousPos:NiagaraPosition=$Particles.Position) {isSetParameters=true}
   &SkeletalMeshLocation(
    Bone / Socket Index=5
    Bone Sampling Mode=::NewEnumerator4
    MaskSpawnGroup=false
    Orientation Sampling=::NewEnumerator1
    Sampled Position Offset=&Add_Vector(A=<10 33 0>, B=$System.MasterRoadOffset)
    Skeletal Mesh=$System.Skeletal Mesh
   )
   /SetVariables(Particles.OrientationGoal:Quat4f=&MultiplyQuaternion(
     Quaternion A=&MultiplyQuaternion(
      Quaternion A=$Particles.SkeletalMeshLocation.SampledRotation
      Quaternion B=&MakeQuatFromAxisAngle(Axis=<0 1 0>, AngleInDegrees=90)
     )
     Quaternion B=&MakeQuatFromAxisAngle(Axis=<1 0 0>, AngleInDegrees=90)
    )) {isSetParameters=true}
   &self:SetRoadBending(CurrentPos:NiagaraPosition=$Particles.Position, MaxBending=250, RestSpeed=1, VelDivider=4) {module=SetRoadBending}
   /SetVariables(Particles.MeshOrientation=$Particles.OrientationGoal) {isSetParameters=true}
   &DynamicMaterialParameters(
    Index 0 Param 1=$Emitter.PanningFloat
    Index 0 Param 2=&Lerp_Float(A=50, B=-20, Alpha=$User.ParaGlideLeanAlpha)
    Index 0 Param 3=$User.ParaGlideLeanAlpha
    Index 0 Param 4=$Particles.Roadbending
    Module.Index0Param0Write=true
    Module.Index0Param1Write=true
    Module.Index0Param2Write=true
    Module.Index0Param3Write=true
   )
   /SetVariables(Particles.Vector 4:Vector4f=&MakeVector4(X=$Particles.Roadbending, Y=0, Z=0, W=0)) {isSetParameters=true}
   &ScaleColor(
    Scale Alpha=&FloatFromCurve(
     FloatCurve=DataInterface() {propertyValues {Curve {keys [{tangentMode=RCTM_User, time=0, value=0}, {tangentMode=RCTM_User, time=0.5, value=1}]}}}
     CurveIndex=$System.Age
    )
    ScaleA=true
    ScaleRGB=true
   ) {module=ScaleColor001}
  ]
  rendererClasses [&NiagaraMeshRendererProperties, &NiagaraSpriteRendererProperties]
  renderers [
   {rendererClass=&NiagaraMeshRendererProperties, rendererIndex=0}
   {rendererClass=&NiagaraSpriteRendererProperties, rendererIndex=1}
  ]
 }
 Road_Dust {
  emitterUpdateScript [
   &EmitterState()
   &SpawnRate(SpawnRate=35, Use Spawn Probability=false)
  ]
  particleSpawnScript [
   &InitializeParticle(
    Color Maximum=<rgba 0.2 0.1242 0.07 0.05>
    Color Minimum=<rgba 0.5 0.3426 0.173 0.025>
    Color Mode=::NewEnumerator2
    Lifetime=&Lerp_Float(A=3, B=2, Alpha=$User.ParaGlideLeanAlpha)
    Sprite Size Max=<120 120>
    Sprite Size Min=<80 80>
    Sprite Size Mode=::NewEnumerator2
    UsePositionOffset=false
    Write Material Random=true
   )
   &InitialMeshOrientation(Mesh Orientation Mode=::NewEnumerator7, Rotation=<0.25 0 0>, Use Rotation Vector=true) {module=InitialMeshOrientation001}
   &InitialMeshOrientation(
    Mesh Orientation Mode=::NewEnumerator7
    Rotation=&RandomRangeVector(Minimum=<0 0 0>, Maximum=<0 1 0>)
    Use Rotation Vector=true
   )
   /SetVariables() {isSetParameters=true, module=SetVariables_#7}
   &SubUVAnimation(
    "Sprite Renderer "=DataInterface() {propertyValues {SpriteRenderer=&self:Road_Dust_18.NiagaraSpriteRendererProperties_1}}
    SubUV Animation Mode:ENiagara_SubUVLookupModeV2=::NewEnumerator2
    UseEndFrame=false
    UseStartFrame=false
   )
  ]
  particleUpdateScript [
   &ParticleState()
   &ScaleSpriteSize(Scale Sprite Size Mode=::NewEnumerator0, Uniform Scale Factor=0.75)
   &SpriteRotationRate(Rotation Rate=&Multiply_Float(A=220, B=&RandomRangeFloat(Minimum=-1, Maximum=1)))
   &self:road_bend_2(Particle Attribute Reader=DataInterface() {propertyValues {EmitterBinding {bindingMode=Other, emitterName=Road}}}) {module=road_bend_2}
   &SkeletalMeshLocation(
    Bone / Socket Index=5
    Bone Sampling Mode=::NewEnumerator4
    MaskSpawnGroup=false
    Sampled Position Offset=&Add_Vector(
     A=&Multiply_Vector(
      A=&Add_Vector(
       A=&Add_Vector(
        A=&Multiply_Vector(
         A=&Lerp_Vector(
          A=&VectorFromCurve(VectorCurve=DataInterface() {propertyValues {
             XCurve {keys [<at 0 250>,<at 1 -250>]}
             YCurve {keys [<at 0 -50 RCIM_Cubic>,<at 0.5 30 RCIM_Cubic>,<at 1 -50 RCIM_Cubic>]}
             ZCurve {keys [<at 0 100>,<at 1 100>]}
            }})
          B=&VectorFromCurve(VectorCurve=DataInterface() {propertyValues {
             XCurve {keys [<at 0 250>,<at 1 -250>]}
             YCurve {keys [<at 0 75 RCIM_Cubic>,<at 0.5 30 RCIM_Cubic>,<at 1 75 RCIM_Cubic>]}
             ZCurve {keys [<at 0 100>,<at 1 100>]}
            }})
          Alpha=$User.ParaGlideLeanAlpha
         )
         B=&MakeCustomVectorFromBool(
          Bool=&SetBoolByIntComparison(
           A=&ModuloInt(A=&ReturnParticleIDAttributes(), B=2)
           B=0
           Comparison Type=::NewEnumerator0
          )
          True Vector=<1 1 0>
          False Vector=<1 1 0>
         )
        )
        B=&Lerp_Vector(
         A=&VectorFromCurve(VectorCurve=DataInterface() {propertyValues {
            XCurve {keys [<at 0 0>,<at 1 0>]}
            YCurve {keys [<at 0 0>,<at 1 0>]}
            ZCurve {keys [
              {
               arriveTangent=-2510.6699
               arriveTangentWeight=363.8912
               interpMode=RCIM_Cubic
               leaveTangent=-2510.6699
               leaveTangentWeight=363.8912
               tangentMode=RCTM_User
               tangentWeightMode=RCTWM_WeightedBoth
               time=0
               value=400
              }
              {
               arriveTangent=0.0001
               arriveTangentWeight=0.7224
               interpMode=RCIM_Cubic
               leaveTangent=0.0001
               leaveTangentWeight=0.7224
               tangentMode=RCTM_User
               tangentWeightMode=RCTWM_WeightedBoth
               time=1
               value=0
              }
             ]}
           }})
         B=&VectorFromCurve(
          VectorCurve=DataInterface() {propertyValues {
            XCurve {keys [<at 0 0>,<at 1 0>]}
            YCurve {keys [<at 0 0>,<at 1 0>]}
            ZCurve {keys [
              {
               arriveTangent=-2510.6699
               arriveTangentWeight=363.8912
               interpMode=RCIM_Cubic
               leaveTangent=-2510.6699
               leaveTangentWeight=363.8912
               tangentMode=RCTM_User
               tangentWeightMode=RCTWM_WeightedBoth
               time=0
               value=400
              }
              {
               arriveTangent=0.0001
               arriveTangentWeight=0.7224
               interpMode=RCIM_Cubic
               leaveTangent=0.0001
               leaveTangentWeight=0.7224
               tangentMode=RCTM_User
               tangentWeightMode=RCTWM_WeightedBoth
               time=1
               value=0
              }
             ]}
           }}
          Scale Curve=<1 1 -1>
         )
         Alpha=&Lerp_Float(
          A=0.5
          B=&RemapRange(Input Value To Remap=$Particles.Roadbending, Input Min=-250, Input Max=250, Output Min=0, Output Max=1)
          Alpha=1
         )
        )
       )
       B=&RandomRangeVector(Minimum=<0 0 -75>, Maximum=<0 0 75>)
      )
      B=$System.MasterRoadScale
     )
     B=$System.MasterRoadOffset
    )
    Skeletal Mesh=$System.Skeletal Mesh
   ) {module=SkeletalMeshLocation001}
   &DynamicMaterialParameters(
    Index 0 Param 1=$Particles.Roadbending
    Module.Index0Param0Write=true
    Module.Index0Param1Write=true
    Module.Index0Param2Write=true
    Module.Index0Param3Write=true
   )
   &ScaleColor(
    Scale Alpha=&FloatFromCurve(
     FloatCurve=DataInterface() {propertyValues {Curve {keys [
         {interpMode=RCIM_Cubic, tangentMode=RCTM_User, time=0, value=0}
         {interpMode=RCIM_Cubic, tangentMode=RCTM_User, time=0.5, value=1}
         {tangentMode=RCTM_User, time=1, value=0}
        ]}}}
     Scale Curve=0.2
    )
    ScaleA=true
    ScaleRGB=true
   )
   /SetVariables(
    Particles.Sand:LinearColor=&MultiplyLinearColor(
     Linear Color=$Particles.Color
     Scale Factor=&MakeFloat4FromVectorAndFloat("Vector (RGB)"=&VectorFromFloat(Value=1), "Float (Alpha)"=2)
    )
    Particles.sand Size:Vector2f=&Multiply_Vector2D(A=$Particles.SpriteSize, B=&Vector2DFromFloat(Value=1.75))
   ) {isSetParameters=true}
   &SubUVAnimation(
    Random Start Frame=true
    "Sprite Renderer "=DataInterface() {propertyValues {SpriteRenderer=&self:Road_Dust_18.NiagaraSpriteRendererProperties_0}}
    SubUV Lookup Index:NiagaraFloat=$Particles.Age
    SubUV Lookup Index Scale=0.5
    UseEndFrame=false
    UseStartFrame=false
   ) {module=SubUVAnimation001}
   /SetVariables() {isSetParameters=true}
   /SetVariables(Particles.Vector 4:Vector4f=&MakeVector4(
     X=&FloatFromCurve(
      FloatCurve=DataInterface() {propertyValues {Curve {keys [<at 0 1>,<at 1 0>]}}}
      CurveIndex=&Add_Float(A=$Particles.NormalizedAge, B=&RandomRangeFloat(Minimum=0, Maximum=1))
     )
     Y=0
     Z=0
     W=0
    )) {isSetParameters=true}
  ]
  rendererClasses [&NiagaraSpriteRendererProperties]
  renderers [
   {rendererClass=&NiagaraSpriteRendererProperties}
   {rendererClass=&NiagaraSpriteRendererProperties, rendererIndex=1}
   {rendererClass=&NiagaraSpriteRendererProperties, rendererIndex=2}
   {rendererClass=&NiagaraSpriteRendererProperties, rendererIndex=3}
  ]
 }
 Sign {
  emitterUpdateScript [
   &EmitterState()
   &SpawnRate(
    SpawnRate=&Divide_Float(A=1, B=&Lerp_Float(A=2.5, B=1.25, Alpha=$User.ParaGlideLeanAlpha))
    Use Spawn Probability=false
   )
  ]
  particleSpawnScript [
   &InitializeParticle(
    Color Mode=::NewEnumerator0
    Lifetime=&Lerp_Float(A=3, B=2, Alpha=$User.ParaGlideLeanAlpha)
    Mesh Scale Mode=::NewEnumerator1
    Mesh Uniform Scale Max=1.1
    Mesh Uniform Scale Min=0.9
    Sprite Size Mode=::NewEnumerator3
    Uniform Sprite Size=72.8114
    UsePositionOffset=false
    Write Material Random=true
   )
   &PartitionParticles()
   &InitialMeshOrientation(Mesh Orientation Mode=::NewEnumerator7, Rotation=<0.25 0 0>, Use Rotation Vector=true) {module=InitialMeshOrientation001}
   &InitialMeshOrientation(
    Mesh Orientation Mode=::NewEnumerator7
    Rotation=&MakeCustomVectorFromBool(
     Bool=&SetBoolByIntComparison(A=$Particles.PartitionParticles.Partition, B=0, Comparison Type=::NewEnumerator4)
     True Vector=&RandomRangeVector(Minimum=<0 -0.25 0>, Maximum=<0 -0.35 0>)
     False Vector=&Add_Vector(A=&RandomRangeVector(Minimum=<0 0.25 0>, Maximum=<0 0.35 0>), B=<0 0.5 0>)
    )
    Use Rotation Vector=true
   )
   /SetVariables() {isSetParameters=true, module=SetVariables_#7}
  ]
  particleUpdateScript [
   &ParticleState()
   &ScaleMeshSize(Scale Factor=&VectorFromFloat(Value=&FloatFromCurve(FloatCurve=DataInterface() {propertyValues {Curve {keys [
          {interpMode=RCIM_Cubic, leaveTangentWeight=0.0333, tangentMode=RCTM_User, time=0, value=0}
          {arriveTangentWeight=0.0498, interpMode=RCIM_Cubic, leaveTangentWeight=0.3986, tangentMode=RCTM_User, time=0.3, value=1}
          {arriveTangentWeight=0.2667, interpMode=RCIM_Cubic, leaveTangentWeight=0.0333, tangentMode=RCTM_User, time=0.7, value=1}
          {tangentMode=RCTM_User, time=1, value=0}
         ]}}})))
   &self:road_bend_2(Particle Attribute Reader=DataInterface() {propertyValues {EmitterBinding {bindingMode=Other, emitterName=Road}}}) {module=road_bend_2}
   &SkeletalMeshLocation(
    Bone / Socket Index=5
    Bone Sampling Mode=::NewEnumerator4
    MaskSpawnGroup=false
    Sampled Position Offset=&Add_Vector(
     A=&Multiply_Vector(
      A=&Add_Vector(
       A=&Multiply_Vector(
        A=&Lerp_Vector(
         A=&VectorFromCurve(VectorCurve=DataInterface() {propertyValues {
            XCurve {keys [<at 0 250>,<at 1 -250>]}
            YCurve {keys [<at 0 -50 RCIM_Cubic>,<at 0.5 30 RCIM_Cubic>,<at 1 -50 RCIM_Cubic>]}
            ZCurve {keys [<at 0 100>,<at 1 100>]}
           }})
         B=&VectorFromCurve(VectorCurve=DataInterface() {propertyValues {
            XCurve {keys [<at 0 250>,<at 1 -250>]}
            YCurve {keys [<at 0 75 RCIM_Cubic>,<at 0.5 30 RCIM_Cubic>,<at 1 75 RCIM_Cubic>]}
            ZCurve {keys [<at 0 100>,<at 1 100>]}
           }})
         Alpha=$User.ParaGlideLeanAlpha
        )
        B=&MakeCustomVectorFromBool(
         Bool=&SetBoolByIntComparison(A=$Particles.PartitionParticles.Partition, B=0, Comparison Type=::NewEnumerator4)
         True Vector=<1 1 1.35>
         False Vector=<1 1 -1.35>
        )
       )
       B=&Lerp_Vector(
        A=&VectorFromCurve(VectorCurve=DataInterface() {propertyValues {
           XCurve {keys [<at 0 0>,<at 1 0>]}
           YCurve {keys [<at 0 0>,<at 1 0>]}
           ZCurve {keys [
             {
              arriveTangent=-2510.6699
              arriveTangentWeight=363.8912
              interpMode=RCIM_Cubic
              leaveTangent=-2510.6699
              leaveTangentWeight=363.8912
              tangentMode=RCTM_User
              tangentWeightMode=RCTWM_WeightedBoth
              time=0
              value=400
             }
             {
              arriveTangent=0.0001
              arriveTangentWeight=0.7224
              interpMode=RCIM_Cubic
              leaveTangent=0.0001
              leaveTangentWeight=0.7224
              tangentMode=RCTM_User
              tangentWeightMode=RCTWM_WeightedBoth
              time=1
              value=0
             }
            ]}
          }})
        B=&VectorFromCurve(
         VectorCurve=DataInterface() {propertyValues {
           XCurve {keys [<at 0 0>,<at 1 0>]}
           YCurve {keys [<at 0 0>,<at 1 0>]}
           ZCurve {keys [
             {
              arriveTangent=-2510.6699
              arriveTangentWeight=363.8912
              interpMode=RCIM_Cubic
              leaveTangent=-2510.6699
              leaveTangentWeight=363.8912
              tangentMode=RCTM_User
              tangentWeightMode=RCTWM_WeightedBoth
              time=0
              value=400
             }
             {
              arriveTangent=0.0001
              arriveTangentWeight=0.7224
              interpMode=RCIM_Cubic
              leaveTangent=0.0001
              leaveTangentWeight=0.7224
              tangentMode=RCTM_User
              tangentWeightMode=RCTWM_WeightedBoth
              time=1
              value=0
             }
            ]}
          }}
         Scale Curve=<1 1 -1>
        )
        Alpha=&Lerp_Float(
         A=0.5
         B=&RemapRange(Input Value To Remap=$Particles.Roadbending, Input Min=-250, Input Max=250, Output Min=0, Output Max=1)
         Alpha=1
        )
       )
      )
      B=$System.MasterRoadScale
     )
     B=$System.MasterRoadOffset
    )
    Skeletal Mesh=$System.Skeletal Mesh
   ) {module=SkeletalMeshLocation001}
   &DynamicMaterialParameters(
    Index 0 Param 1=$Particles.Roadbending
    Module.Index0Param0Write=true
    Module.Index0Param1Write=true
    Module.Index0Param2Write=true
    Module.Index0Param3Write=true
   )
  ]
  rendererClasses [&NiagaraMeshRendererProperties, &NiagaraSpriteRendererProperties]
  renderers [
   {rendererClass=&NiagaraMeshRendererProperties, rendererIndex=0}
   {rendererClass=&NiagaraSpriteRendererProperties, rendererIndex=1}
  ]
 }
 arm_sand {
  emitterUpdateScript [
   &EmitterState(Life Cycle Mode=::NewEnumerator1, Loop Behavior=::NewEnumerator1, Loop Delay=0.5, Loop Duration Mode=::NewEnumerator1, UseLoopDelay=true)
   &SpawnBurst_Instantaneous(Spawn Count=4, Use Spawn Probability=false, UseLoopCountLimit=false)
   &SpawnBurst_Instantaneous(Spawn Count=2, Spawn Group=1, Use Spawn Probability=false, UseLoopCountLimit=false) {module=SpawnBurst_Instantaneous001}
   &SpawnBurst_Instantaneous(Spawn Count=2, Use Spawn Probability=false, UseLoopCountLimit=false) {enabled=false, module=SpawnBurst_Instantaneous002}
  ]
  particleSpawnScript [
   &PartitionParticles(Number Of Partitions=&MaskIntBySpawnGroup(Masked Int=2, Passthrough Int=2, Spawn Group Mask=0))
   &InitializeParticle(
    Color=<rgba 0.075 0.0522 0.0277 10>
    Lifetime=2
    Mesh Renderer Array Visibility Mode=::NewEnumerator1
    Mesh Renderer Info=DataInterface() {propertyValues {MeshRenderer=&self:arm_sand_23.NiagaraMeshRendererProperties_0}}
    Mesh Renderer Mesh Index:NiagaraInt32=&SpawnGroup()
    Mesh Scale=&SelectVectorFromArray(
     Direct Array Index=$Particles.PartitionParticles.Partition
     Vector Selection Array=DataInterface() {propertyValues {FloatData [<1.5 1.5 0.55>, <1.5 1.5 0.55>, <1 1 0.25>, <1 1 0.25>]}}
    )
    Mesh Scale Mode=::NewEnumerator4
    Sprite Size Mode=::NewEnumerator3
    Uniform Sprite Size=50
    UsePositionOffset=false
    Write Material Random=true
   )
   &InitialMeshOrientation(
    Mesh Axis To Orient=<0 0 -1>
    Mesh Orientation Mode=::NewEnumerator2
    Rotation=&Lerp_Vector(A=<0 0 0>, B=<0.5 0 0>, Alpha=&MakeFloatFromInt(Int=$Particles.PartitionParticles.Partition))
    Rotation Coordinate Space=::NewEnumerator0
    Use Rotation Vector=true
   )
   /SetVariables(Particles.RendererTag:NiagaraInt32=&MaskIntBySpawnGroup(Masked Int=0, Passthrough Int=1, Spawn Group Mask=0)) {isSetParameters=true}
  ]
  particleUpdateScript [
   &ParticleState(
    Kill Particles When Lifetime Has Elapsed=false
    Let Infinitely Lived Particles Die When Emitter Deactivates=true
    Loop Particles Lifetime=true
   )
   &SkeletalMeshLocation(
    Bone / Socket Index=$Particles.PartitionParticles.Partition
    Bone Sampling Mode=::NewEnumerator4
    "Interpolate Mesh Sampling (Update)"=true
    MaskSpawnGroup=false
    Orientation Sampling=::NewEnumerator1
    Sampled Position Offset=&SelectVectorFromArray(
     Direct Array Index=$Particles.PartitionParticles.Partition
     Vector Selection Array=DataInterface() {propertyValues {FloatData [<10 0 0>, <-10 0 0>, <2 0 0>, <-5 0 0>]}}
    )
    Skeletal Mesh=$System.Skeletal Mesh
   )
   &self:road_bend_2(Particle Attribute Reader=DataInterface() {propertyValues {EmitterBinding {bindingMode=Other, emitterName=road}}}) {module=road_bend_2}
   &ScaleMeshSize(Scale Factor=&VectorFromFloat(Value=&Lerp_Float(
      A=&FloatFromCurve(
       FloatCurve=DataInterface() {propertyValues {Curve {keys [{arriveTangent=-0.0746, interpMode=RCIM_Cubic, leaveTangent=-0.0746, tangentMode=RCTM_User, time=-50, value=1}, <at -10 0 RCIM_Cubic>]}}}
       CurveIndex=$Particles.Roadbending
      )
      B=&FloatFromCurve(
       FloatCurve=DataInterface() {propertyValues {Curve {keys [<at 10 0 RCIM_Cubic>, {arriveTangent=0.0504, interpMode=RCIM_Cubic, leaveTangent=0.0504, tangentMode=RCTM_User, time=50, value=1}]}}}
       CurveIndex=$Particles.Roadbending
      )
      Alpha=&OneMinusFloat(Float=&MakeFloatFromInt(Int=$Particles.PartitionParticles.Partition))
     ))) {enabled=false, module=ScaleMeshSize002}
   /SetVariables(Particles.Position=&Add_Vector(
     A=&ConvertPositionToVector(Input Position=$Particles.Position, Convert To Absolute Large World Vector=false)
     B=&Add_Vector(A=&Multiply_VectorByFloat(Vector=$System.DiveVelocityScale, Float=10), B=<0 0 0>)
    )) {isSetParameters=true}
   &UpdateMeshOrientation(
    Facing Direction=$System.DiveVelocityScale
    Facing Interpolation=false
    Orientation Method=::NewEnumerator1
    Override Quat=false
    Side Direction:Vector3f=$Engine.Owner.SystemXAxis
    Stabilization Mode:ENiagara_OrientVectorCount=::NewEnumerator2
    TransformFacing=false
    TransformSide=false
   ) {module=UpdateMeshOrientation001}
   &UpdateMeshOrientation(
    Rotation Rate=&MaskFloatBySpawnGroup(Masked Float=0, Passthrough Float=&RandomRangeFloat(Minimum=1, Maximum=1.6))
    Rotation Vector=<0 0 1>
   )
   &ScaleMeshSize(Scale Factor=&Multiply_Vector(
     A=&Lerp_Vector(A=<1.2 1.2 2>, B=<1 1 1.5>, Alpha=$System.FN_Integrator001.Value)
     B=&MakeVector(
      X=1
      Y=1
      Z=&FloatFromCurve(
       FloatCurve=DataInterface() {propertyValues {Curve {keys [<at 0 0 RCIM_Cubic>, {arriveTangent=4.1531, interpMode=RCIM_Cubic, leaveTangent=4.1531, tangentMode=RCTM_User, time=0.5, value=1}]}}}
       CurveIndex=$Emitter.Age
      )
     )
    ))
   &ScaleMeshSize(Scale Factor=&LerpPosition(
     A=&SelectVectorFromArray(
      Direct Array Index=$Particles.PartitionParticles.Partition
      Vector Selection Array=DataInterface() {propertyValues {FloatData [<1 1 1>, <1 1 1>, <1 1 1>, <1 1 1>]}}
     )
     B=&SelectVectorFromArray(
      Direct Array Index=$Particles.PartitionParticles.Partition
      Vector Selection Array=DataInterface() {propertyValues {FloatData [<1.5 1.5 1.5>, <1.5 1.5 1.5>, <1.5 1.5 2>, <1.5 1.5 2>]}}
     )
     Alpha=$System.Dive
    )) {module=ScaleMeshSize001}
   &ScaleColor(
    Scale Alpha=&FloatFromCurve(
     FloatCurve=DataInterface() {propertyValues {Curve {keys [<at 0.5 0 RCIM_Cubic>, {arriveTangent=4.1531, interpMode=RCIM_Cubic, leaveTangent=4.1531, tangentMode=RCTM_User, time=1, value=1}]}}}
     CurveIndex=$Emitter.Age
     Scale Curve=0.5
    )
    Scale RGB=&Multiply_VectorByFloat(Vector=&MakeVectorFromLinearColorRGB(Color=<rgba 1 1 1 1>), Float=0.25)
    ScaleA=true
    ScaleRGB=true
   ) {module=ScaleColor001}
   /SetVariables(Particles.SpiralColor:LinearColor=&MultiplyLinearColor(
     Linear Color=$Particles.Color
     Scale Factor=&MakeVector4_FromLinearColor(LinearColor=&MultiplyLinearColor(
       Linear Color=<rgba 0.276 0.276 0.276 0.1>
       Scale Factor=&MakeFloat4FromVectorAndFloat(
        "Vector (RGB)"=&VectorFromFloat(Value=1)
        "Float (Alpha)"=&FN_FrontEndFloatModifier(
         Gameplay=&Lerp_Float(
          A=&FloatFromCurve(
           FloatCurve=DataInterface() {propertyValues {Curve {keys [{arriveTangent=-0.0746, interpMode=RCIM_Cubic, leaveTangent=-0.0746, tangentMode=RCTM_User, time=-50, value=1}, <at -10 0 RCIM_Cubic>]}}}
           CurveIndex=$Particles.Roadbending
          )
          B=&FloatFromCurve(
           FloatCurve=DataInterface() {propertyValues {Curve {keys [<at 10 0 RCIM_Cubic>, {arriveTangent=0.0504, interpMode=RCIM_Cubic, leaveTangent=0.0504, tangentMode=RCTM_User, time=50, value=1}]}}}
           CurveIndex=$Particles.Roadbending
          )
          Alpha=&OneMinusFloat(Float=&MakeFloatFromInt(Int=$Particles.PartitionParticles.Partition))
         )
         Front End=0
        )
       )
      ))
    )) {isSetParameters=true}
   &InterpolateOverTime(
    Rate Of Change=12.5
    Target Vector Value:Vector3f=&AddVectorToPosition(Position=$StackContext.SkeletalMeshLocation.SampledPosition, Vector=<0 0 250>)
    Variable Type=::NewEnumerator2
   ) {module=InterpolateOverTime001}
   /SetVariables(Particles.TargetPosition:NiagaraPosition=$StackContext.InterpolateOverTime001.InterpolateOverTime_V3_.Moving Average) {isSetParameters=true}
   /SetVariables(Particles.TargetDelta:Vector3f=&Subtract_Vector(
     A=Linked() {linkedVariable {name=Particles.TargetPosition, type=NiagaraPosition}}
     B=Linked() {linkedVariable {name=Particles.Position, type=NiagaraPosition}}
    )) {isSetParameters=true}
   &DynamicMaterialParameters(
    Index 0 Param 1=&RandomRangeFloat(Minimum=0, Maximum=1)
    Index 0 Param 2=&RandomRangeFloat(Minimum=0, Maximum=1.5)
    Index 1 Param 1=&MakeFloatFromVector(Vector=$Particles.TargetDelta, Channel=::NewEnumerator0)
    Index 1 Param 2=&MakeFloatFromVector(Vector=$Particles.TargetDelta, Channel=::NewEnumerator1)
    Index 1 Param 3=&OneMinusFloat(Float=$System.Dive)
    Index 1 Param 4=128
    Index 2 Param 1=111
    Index 2 Param 3=0.2
    Index 2 Param 4=&MaskFloatBySpawnGroup(Masked Float=-2, Passthrough Float=1, Spawn Group Mask=1)
    Module.Param0WriteEnabled=true
    Module.Param0WriteEnabled1=true
    Module.Param0WriteEnabled2=true
    Module.Param1WriteEnabled=true
    Module.Param1WriteEnabled1=true
    Module.Param1WriteEnabled2=true
    Module.Param2WriteEnabled=false
    Module.Param2WriteEnabled1=true
    Module.Param2WriteEnabled2=true
    Module.Param3WriteEnabled=true
    Module.Param3WriteEnabled1=true
    Module.Param3WriteEnabled2=true
    Write Parameter Index 1=true
    Write Parameter Index 2=true
   )
   &ScaleColor(
    Scale Alpha=&FN_FrontEndFloatModifier(
     Gameplay=&Lerp_Float(
      A=&FloatFromCurve(
       FloatCurve=DataInterface() {propertyValues {Curve {keys [{arriveTangent=-0.0746, interpMode=RCIM_Cubic, leaveTangent=-0.0746, tangentMode=RCTM_User, time=-50, value=1}, <at -10 0 RCIM_Cubic>]}}}
       CurveIndex=$Particles.Roadbending
      )
      B=&FloatFromCurve(
       FloatCurve=DataInterface() {propertyValues {Curve {keys [<at 10 0 RCIM_Cubic>, {arriveTangent=0.0504, interpMode=RCIM_Cubic, leaveTangent=0.0504, tangentMode=RCTM_User, time=50, value=1}]}}}
       CurveIndex=$Particles.Roadbending
      )
      Alpha=&OneMinusFloat(Float=&MakeFloatFromInt(Int=$Particles.PartitionParticles.Partition))
     )
     Front End=0
    )
    ScaleA=true
    ScaleRGB=false
   )
  ]
  rendererClasses [&NiagaraMeshRendererProperties]
  renderers [
   {rendererClass=&NiagaraMeshRendererProperties, rendererIndex=0}
   {rendererClass=&NiagaraMeshRendererProperties, rendererIndex=1}
  ]
 }
}
summary {
 systemName=NS_Contrail_ShiverFlame
 emitters [
  {emitterName=Road, rendererClasses [&NiagaraMeshRendererProperties, &NiagaraSpriteRendererProperties]}
  {emitterName=arm_sand, rendererClasses [&NiagaraMeshRendererProperties]}
  {emitterName=Leg_Sand, rendererClasses [&NiagaraSpriteRendererProperties]}
  {emitterName=Barrel002, rendererClasses [&NiagaraMeshRendererProperties]}
  {
   emitterName=Barrior_Parts_Burst
   rendererClasses [&NiagaraSpriteRendererProperties, &NiagaraMeshRendererProperties]
  }
  {emitterName=Sign, rendererClasses [&NiagaraMeshRendererProperties, &NiagaraSpriteRendererProperties]}
  {emitterName=Barrior, rendererClasses [&NiagaraMeshRendererProperties, &NiagaraSpriteRendererProperties]}
  {emitterName=Barrior_death_sand, rendererClasses [&NiagaraSpriteRendererProperties]}
  {emitterName=Road_Dust, rendererClasses [&NiagaraSpriteRendererProperties]}
  {
   bEnabled=false
   emitterName=Empty
   rendererClasses [&NiagaraSpriteRendererProperties, &NiagaraMeshRendererProperties]
  }
  {emitterName=NE_FN_PhysMatDebris, rendererClasses [&NiagaraMeshRendererProperties]}
 ]
 userVariables [
  User.bIsFrontEnd {type=NiagaraBool}
  User.bIsFrontEndPreview {type=NiagaraBool}
  User.ForwardDot {type=NiagaraFloat, defaultValue=0, description=""}
  User.ParaGlideLeanAlpha {type=NiagaraFloat, defaultValue=1, description=""}
  User.PhysicalSurface {type=NiagaraInt32}
  User.PlayerSpeed {type=NiagaraFloat, defaultValue=0, description=""}
  User.RibbonIDOffset {type=NiagaraInt32}
  User.RightDot {type=NiagaraFloat, defaultValue=0.25, description=""}
 ]
}
